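# Physics asset for the female medium baseball kitbash hat face accessory, simulating two dynamic ponytails, left and right, that hang off the head.
# Each ponytail is a two-bone chain of capsule bodies constrained to a head capsule and to each other.
# All joints share moderate swing limits and very tight twist limits, so the tails can sway but barely roll.
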PhysicsAsset /BRCosmetics/Accessories/Hats/F_MED_Baseball_Kitbash_FaceAcc/Meshes/F_MED_Baseball_Kitbash_FaceAcc_Dyn
bodies {
 dyn_ponytail_l [{shapeType=Capsule, length=3.6936, radius=3.8697, center=<3.6 0 0>, rotation=<rot 90 0 0>}]
 dyn_ponytail_l_02 [{shapeType=Capsule, length=3.6936, radius=3.8697, center=<2 0 0>, rotation=<rot 90 0 0>}]
 dyn_ponytail_r [{shapeType=Capsule, length=3.678, radius=3.7929, center=<-3.6 0 0>, rotation=<rot 90 0 0>}]
 dyn_ponytail_r_02 [{shapeType=Capsule, length=3.678, radius=3.7929, center=<-2 0 0>, rotation=<rot 90 0 0>}]
 head [{shapeType=Capsule, length=3.192, radius=2.2155, rotation=<rot 90 0 0>}]
}
constraints [
 {bone1Name=dyn_ponytail_r, bone2Name=head, swing1LimitDegrees=45, swing2LimitDegrees=45, twistLimitDegrees=2}
 {bone1Name=dyn_ponytail_l, bone2Name=head, swing1LimitDegrees=45, swing2LimitDegrees=45, twistLimitDegrees=2}
 {bone1Name=dyn_ponytail_l_02, bone2Name=dyn_ponytail_l, swing1LimitDegrees=45, swing2LimitDegrees=45, twistLimitDegrees=2}
 {bone1Name=dyn_ponytail_r_02, bone2Name=dyn_ponytail_r, swing1LimitDegrees=45, swing2LimitDegrees=45, twistLimitDegrees=2}
]
